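# Niagara system for the despawn effect of the PawJasmine player emote, judging by its name and path.
# It holds three sets of Sparkles, Airglow and Smoke emitters, with LWE names suggesting lightweight emitters.
# SystemState is the only listed module, and no renderers are listed.
# It exposes user parameters for a vector, normalized notify progress and a ribbon ID offset, with no visible links to the emitters.
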NiagaraSystem /BRCosmetics/Animation/Game/MainPlayer/Emotes/PawJasmine/FX/NS_Emote_PawJasmine_Despawn
dependencies {usedDataInterfaces [], usedDynamicInputs [], usedModules [&SystemState]}
emitters {
 AirglowLWE {rendererClasses [], renderers []}
 AirglowLWE001 {rendererClasses [], renderers []}
 AirglowLWE002 {rendererClasses [], renderers []}
 SmokeLWE {rendererClasses [], renderers []}
 SmokeLWE001 {rendererClasses [], renderers []}
 SmokeLWE002 {rendererClasses [], renderers []}
 SparklesLWE {rendererClasses [], renderers []}
 SparklesLWE001 {rendererClasses [], renderers []}
 SparklesLWE002 {rendererClasses [], renderers []}
}
summary {
 systemName=NS_Emote_PawJasmine_Despawn
 emitters [
  {emitterName=SparklesLWE, rendererClasses []}
  {emitterName=AirglowLWE, rendererClasses []}
  {emitterName=SmokeLWE, rendererClasses []}
  {emitterName=SparklesLWE001, rendererClasses []}
  {emitterName=AirglowLWE001, rendererClasses []}
  {emitterName=SmokeLWE001, rendererClasses []}
  {emitterName=SparklesLWE002, rendererClasses []}
  {emitterName=AirglowLWE002, rendererClasses []}
  {emitterName=SmokeLWE002, rendererClasses []}
 ]
 userVariables [
  User.NewVector3f {type=Vector3f, description="", defaultValue {struct=Vector3f, value=<0 0 0>}}
  User.NormalizedNotifyProgress {type=NiagaraFloat, defaultValue=0, description=""}
  User.RibbonIDOffset {type=NiagaraInt32}
 ]
}
