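# Niagara system for the PawJasmine emote that plays repeated short flashes of light, with loop counts driven by User.burst.
# LightRays and SimpleSpriteBurst fire three timed bursts per loop, partitioning particles across three sockets from a system-level Socket Reader.
# LightRays layers three sprite renderers and a light renderer using separate flash, glow and streak attributes, while SimpleSpriteBurst adds small sparkles.
# ParticleDecal loops half as often, scattering fading sprites over a User.Size box surface.
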NiagaraSystem /BRCosmetics/Animation/Game/MainPlayer/Emotes/PawJasmine/FX/NS_Emote_PawJasmine_Flashes
dependencies {
 usedDataInterfaces [
  &NiagaraDataInterfaceArrayInt32
  &NiagaraDataInterfaceVectorCurve
  &NiagaraDataInterfaceColorCurve
  &NiagaraDataInterfaceCurve
  &NiagaraDataInterfaceSocketReader
 ]
 usedDynamicInputs []
 usedModules [
  &self:SystemSpawnScript.SystemScriptSource.SystemScriptGraph.NiagaraNodeAssignment_1.SetVariables_#1
  &SystemState
 ]
}
emitters {
 LightRays {
  emitterUpdateScript [
   &EmitterState(
    Life Cycle Mode=::NewEnumerator1
    Loop Behavior=::NewEnumerator2
    Loop Count:NiagaraInt32=&MakeIntFromFloat(Float=$User.burst, Conversion Mode=::NewEnumerator2)
    Loop Duration=0.32
    Recalculate Duration Each Loop=true
    UseLoopDelay=false
   )
   &SpawnBurst_Instantaneous(
    Spawn Time=&RandomRangeFloat(Minimum=0, Maximum=0.2, Evaluation Type=::NewEnumerator1, Random Seed=13, Fixed Random Seed=true)
    Use Spawn Probability=false
    UseLoopCountLimit=false
   ) {module=SpawnBurst_Instantaneous001}
   &SpawnBurst_Instantaneous(
    Spawn Time=&RandomRangeFloat(Minimum=0, Maximum=0.2, Evaluation Type=::NewEnumerator1, Random Seed=13, Fixed Random Seed=true)
    Use Spawn Probability=false
    UseLoopCountLimit=false
   ) {module=SpawnBurst_Instantaneous002}
   &SpawnBurst_Instantaneous(
    Spawn Time=&RandomRangeFloat(Minimum=0, Maximum=0.2, Evaluation Type=::NewEnumerator1, Random Seed=13, Fixed Random Seed=true)
    Use Spawn Probability=false
    UseLoopCountLimit=false
   )
  ]
  particleSpawnScript [
   &InitializeParticle(
    Color=<rgba 0.7694 0.9381 1 0.2348>
    Lifetime=0.3
    Mass Mode=::NewEnumerator1
    Mesh Scale=<0.07 0.07 0.221>
    Mesh Scale Mode=::NewEnumerator4
    Sprite Size Mode=::NewEnumerator3
    Uniform Sprite Size=20
    UsePositionOffset=false
    Write Material Random=true
   )
   &PartitionParticles(Number Of Partitions=3)
   &InitialMeshOrientation(Mesh Orientation Mode=::NewEnumerator7, Rotation=<0 -0.25 0>, Use Rotation Vector=true)
   &ShapeLocation(MaskSpawnGroup=true, Override Seed=false, Spawn Group Mask=1, Sphere Radius=198.2375, UseOwnerScale=true)
   /SetVariables(Particles.VisibilityTag=&SpawnGroup()) {isSetParameters=true}
   &SocketLocation(
    SamplePosOffset=true
    Sampled Orientation Coordinate Space:ENiagaraCoordinateSpace=::Local
    Socket Index=&SelectIntFromArray(Direct Array Index=$Particles.PartitionParticles.Partition, Int Selection Array=DataInterface() {propertyValues {IntData [0,1,2]}})
    Socket Reader=$System.Socket Reader
    Socket Sampling Mode=::NewEnumerator2
    VelocitySpeedLimit=false
   )
  ]
  particleUpdateScript [
   &ParticleState()
   &ScaleMeshSize(Scale Factor=&VectorFromCurve(VectorCurve=DataInterface() {propertyValues {XCurve {keys [<at 0 0>,<at 0.5 1>,<at 1 0>]}, YCurve {keys [<at 0 0>,<at 0.5 1>,<at 1 0>]}, ZCurve {keys [<at 0 1>,<at 1 1>]}}}))
   &SolveForcesAndVelocity(Clamp Velocity=false, Limit Acceleration=false)
   /SetVariables(
    Particles.FlashCameraOffset=15
    Particles.FlashColor:LinearColor=&ColorFromCurve(ColorCurve=DataInterface() {propertyValues {
       AlphaCurve {keys [<at 0 0.829 RCIM_Cubic>,<at 0.3541 0>,<at 1 0 RCIM_Cubic>]}
       BlueCurve {keys [<at 0 1 RCIM_Cubic>,<at 0.8761 0 RCIM_Cubic>]}
       GreenCurve {keys [<at 0 1 RCIM_Cubic>,<at 0.8761 0 RCIM_Cubic>]}
       RedCurve {keys [<at 0 1 RCIM_Cubic>,<at 0.8761 0 RCIM_Cubic>]}
      }})
    Particles.SpriteCameraOffset=25
   ) {isSetParameters=true}
   &FN_Light_Attributes(Create Color Attribute=true, Create Position Attribute=true, Create Radius Attribute=true, Position=<0 10 0>, Radius=25)
   /SetVariables(
    Particles.NikeAlignment:Vector3f=<0 1 0>
    Particles.NikeCameraOffset=9
    Particles.NikeColor:LinearColor=&ColorFromCurve(ColorCurve=DataInterface() {propertyValues {
       AlphaCurve {keys [<at 0.1921 0>,<at 0.2367 0.8 RCIM_Cubic>,<at 1 0 RCIM_Cubic>]}
       BlueCurve {keys [<at 0.3636 1.2 RCIM_Cubic>,<at 0.7094 0.1667>,<at 1 0 RCIM_Cubic>]}
       GreenCurve {keys [<at 0.3636 1.2 RCIM_Cubic>,<at 0.7094 0.1667>,<at 1 0 RCIM_Cubic>]}
       RedCurve {keys [<at 0.3636 1.2 RCIM_Cubic>,<at 0.7094 0.1667>,<at 1 0 RCIM_Cubic>]}
      }})
    Particles.NikeFacing:Vector3f=<1 0 0>
    Particles.NikePos:NiagaraPosition=$Output.SocketLocation001.SampledPosition
    Particles.NikeSize:Vector2f=&Multiply_Vector2D_ByFloat(Float=7, Vector2D=<2.8 1>)
   ) {isSetParameters=true}
   /SetVariables(Particles.FlashSize:Vector2f=&Multiply_Vector2D_ByFloat(
     Float=&FloatFromCurve(FloatCurve=DataInterface() {propertyValues {Curve {keys [<at 0 0>,<at 0.3869 1>,<at 0.6434 1>,<at 1 0>]}}}, Scale Curve=3.7)
     Vector2D=<1 1>
    )) {isSetParameters=true}
   /SetVariables(
    Particles.GlowColor=&ColorFromCurve(ColorCurve=DataInterface() {propertyValues {
       AlphaCurve {keys [
         <at 0.1921 0>
         {arriveTangent=1.3554, interpMode=RCIM_Cubic, leaveTangent=1.3554, time=0.2367, value=0.3}
         <at 0.288 0.13>
         <at 1 0 RCIM_Cubic>
        ]}
       BlueCurve {keys [<at 0.3636 1.2 RCIM_Cubic>,<at 0.7094 0.1667>,<at 1 0 RCIM_Cubic>]}
       GreenCurve {keys [<at 0.3636 1.2 RCIM_Cubic>,<at 0.7094 0.1667>,<at 1 0 RCIM_Cubic>]}
       RedCurve {keys [<at 0.3636 1.2 RCIM_Cubic>,<at 0.7094 0.1667>,<at 1 0 RCIM_Cubic>]}
      }})
    Particles.GlowSize=<60 60>
   ) {isSetParameters=true}
  ]
  rendererClasses [&NiagaraSpriteRendererProperties, &NiagaraLightRendererProperties]
  renderers [
   {rendererClass=&NiagaraSpriteRendererProperties}
   {rendererClass=&NiagaraLightRendererProperties, rendererIndex=1}
   {rendererClass=&NiagaraSpriteRendererProperties, rendererIndex=2}
   {rendererClass=&NiagaraSpriteRendererProperties, rendererIndex=3}
  ]
 }
 ParticleDecal {
  emitterUpdateScript [
   &EmitterState(
    Life Cycle Mode=::NewEnumerator1
    Loop Behavior=::NewEnumerator2
    Loop Count:NiagaraInt32=&MakeIntFromFloat(Float=&Divide_Float(A=$User.burst, B=2), Conversion Mode=::NewEnumerator2)
    Loop Duration=0.3
    UseLoopDelay=false
   )
   &SpawnBurst_Instantaneous(Spawn Count=2, Use Spawn Probability=false, UseLoopCountLimit=false)
  ]
  particleSpawnScript [
   &InitializeParticle(
    Color=<rgba 5 5 5 0.0783>
    Lifetime=0.2
    Mesh Scale Mode=::NewEnumerator3
    Mesh Uniform Scale=2
    Sprite Rotation Angle=&RandomRangeFloat(Minimum=-19.1044, Maximum=19.5334)
    Sprite Rotation Mode=::NewEnumerator0
    Sprite Size=&Multiply_Vector2D_ByFloat(Float=&RandomRangeFloat(Minimum=10, Maximum=14), Vector2D=<4.0889 1.9219>)
    Sprite Size Mode=::NewEnumerator4
    UsePositionOffset=true
    Write Material Random=true
   )
   &CameraOffset(Camera Offset Amount=30)
   &DynamicMaterialParameters(Module.Index0Param0Write=true, Module.Index0Param1Write=true, Module.Index0Param2Write=true, Module.Index0Param3Write=true)
   &ShapeLocation(
    Box Size=$User.Size
    MaskSpawnGroup=false
    Offset Mode=::NewEnumerator0
    Override Seed=false
    Shape Primitive=::NewEnumerator2
    Surface Expansion Mode:ENiagaraExpansionMode=::NewEnumerator2
    Surface Only=true
    Surface Only Band Thickness=0.064
    UseOwnerScale=true
   )
  ]
  particleUpdateScript [
   &ParticleState()
   &ScaleColor(
    Linear Color Curve=DataInterface() {propertyValues {
      AlphaCurve {keys [<at 0 1>,<at 1 0>]}
      BlueCurve {keys [<at 0 1>,<at 1 0>]}
      GreenCurve {keys [<at 0 1>,<at 1 0>]}
      RedCurve {keys [<at 0 1>,<at 1 0>]}
     }}
    Scale Mode=::NewEnumerator2
   )
   /SetVariables(
    Particles.CircleColor:LinearColor=&ColorFromCurve(ColorCurve=DataInterface() {propertyValues {
       AlphaCurve {keys [<at 0 0.233 RCIM_Cubic>,<at 1 0 RCIM_Cubic>]}
       BlueCurve {keys [<at 0 0.2656 RCIM_Cubic>,<at 1 0 RCIM_Cubic>]}
       GreenCurve {keys [<at 0 0.2656 RCIM_Cubic>,<at 1 0 RCIM_Cubic>]}
       RedCurve {keys [<at 0 0.2656 RCIM_Cubic>,<at 1 0 RCIM_Cubic>]}
      }})
    Particles.CircleSize:Vector2f=&Multiply_Vector2D_ByFloat(Float=50, Vector2D=<1 1>)
   ) {isSetParameters=true}
  ]
  rendererClasses [&NiagaraSpriteRendererProperties]
  renderers [{rendererClass=&NiagaraSpriteRendererProperties}]
 }
 SimpleSpriteBurst {
  emitterUpdateScript [
   &EmitterState(
    Life Cycle Mode=::NewEnumerator1
    Loop Behavior=::NewEnumerator2
    Loop Count:NiagaraInt32=&MakeIntFromFloat(Float=$User.burst, Conversion Mode=::NewEnumerator2)
    Loop Duration=0.35
    Recalculate Duration Each Loop=true
    UseLoopDelay=false
   )
   &SpawnBurst_Instantaneous(
    Spawn Count=&RandomRangeInt(Minimum=1, Maximum=2)
    Spawn Time=&RandomRangeFloat(Minimum=0, Maximum=0.3, Evaluation Type=::NewEnumerator1, Random Seed=13, Fixed Random Seed=true)
    Use Spawn Probability=false
    UseLoopCountLimit=false
   ) {module=SpawnBurst_Instantaneous001}
   &SpawnBurst_Instantaneous(
    Spawn Count=&RandomRangeInt(Minimum=1, Maximum=2)
    Spawn Time=&RandomRangeFloat(Minimum=0, Maximum=0.3, Evaluation Type=::NewEnumerator1, Random Seed=13, Fixed Random Seed=true)
    Use Spawn Probability=false
    UseLoopCountLimit=false
   ) {module=SpawnBurst_Instantaneous002}
   &SpawnBurst_Instantaneous(
    Spawn Count=&RandomRangeInt(Minimum=1, Maximum=2)
    Spawn Time=&RandomRangeFloat(Minimum=0, Maximum=0.3, Evaluation Type=::NewEnumerator1, Random Seed=13, Fixed Random Seed=true)
    Use Spawn Probability=false
    UseLoopCountLimit=false
   )
  ]
  particleSpawnScript [
   &PartitionParticles(Number Of Partitions=3)
   &InitializeParticle(
    Color Maximum=<rgba 0.8225 1 0.7719 0.7565>
    Color Minimum=<rgba 0.4531 0.5305 1 0.9217>
    Color Mode=::NewEnumerator2
    Lifetime=&RandomRangeFloat(Minimum=0.1, Maximum=0.15)
    Sprite Size Mode=::NewEnumerator3
    Uniform Sprite Size=&RandomRangeFloat(Minimum=2, Maximum=4)
    UsePositionOffset=false
    Write Material Random=true
   )
   &SocketLocation(
    Position Sampling=::NewEnumerator1
    SamplePosOffset=true
    Sampled Orientation Coordinate Space:ENiagaraCoordinateSpace=::Local
    Socket Index=&SelectIntFromArray(Direct Array Index=$Particles.PartitionParticles.Partition, Int Selection Array=DataInterface() {propertyValues {IntData [0,1,2]}})
    Socket Reader=$System.Socket Reader
    Socket Sampling Mode=::NewEnumerator2
    VelocitySpeedLimit=false
   ) {module=SocketLocation001}
   &ShapeLocation(MaskSpawnGroup=false, Override Seed=false, Shape Origin=$Output.SocketLocation001.SampledPosition, Sphere Radius=3, UseOwnerScale=true)
   &CameraOffset(Camera Offset Amount=20)
  ]
  particleUpdateScript [
   &ParticleState()
   &ScaleSpriteSize(Uniform Curve Sprite Scale=DataInterface() {propertyValues {Curve {
       defaultValue=0
       keys [
        {arriveTangent=9.0725, interpMode=RCIM_Cubic, leaveTangent=9.0725, tangentMode=RCTM_User, time=0, value=0}
        {arriveTangent=-2.5249, interpMode=RCIM_Cubic, leaveTangent=-2.5249, tangentMode=RCTM_User, time=0.1408, value=0.9847}
        {arriveTangent=-0.0405, interpMode=RCIM_Cubic, leaveTangent=-0.0405, tangentMode=RCTM_User, time=1, value=0}
       ]
      }}})
   &ScaleColor(
    Scale Alpha=&FloatFromCurve(FloatCurve=DataInterface() {propertyValues {Curve {keys [<at -0.0046 1.0113 RCIM_Cubic>,<at 1.0011 0.002 RCIM_Cubic>]}}})
    ScaleA=false
    ScaleRGB=false
   )
   &SolveForcesAndVelocity(Clamp Velocity=false, Limit Acceleration=false)
  ]
  rendererClasses [&NiagaraSpriteRendererProperties]
  renderers [{rendererClass=&NiagaraSpriteRendererProperties}]
 }
}
summary {
 systemName=NS_Emote_PawJasmine_Flashes
 emitters [
  {emitterName=LightRays, rendererClasses [&NiagaraSpriteRendererProperties, &NiagaraLightRendererProperties]}
  {emitterName=ParticleDecal, rendererClasses [&NiagaraSpriteRendererProperties]}
  {emitterName=SimpleSpriteBurst, rendererClasses [&NiagaraSpriteRendererProperties]}
 ]
 userVariables [
  User.burst {type=NiagaraFloat, defaultValue=0, description=""}
  User.Size {type=Vector3f, description="", defaultValue {struct=Vector3f, value=<0 0 0>}}
 ]
}
