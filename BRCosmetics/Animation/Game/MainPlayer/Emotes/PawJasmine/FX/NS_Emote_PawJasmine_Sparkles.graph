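# Niagara system that supplies the sparkle effect for the PawJasmine player emote.
# It holds two SparklesLWE emitters, likely lightweight given the name and empty renderer lists, plus a Decal emitter.
# System-level variable assignments draw on curve colors, float multiplies, time-of-day color and scalar inputs, and float-comparison bool switches.
# It exposes two float user parameters, Bool and rotation.
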NiagaraSystem /BRCosmetics/Animation/Game/MainPlayer/Emotes/PawJasmine/FX/NS_Emote_PawJasmine_Sparkles
dependencies {
 usedDataInterfaces [&NiagaraDataInterfaceColorCurve]
 usedDynamicInputs [
  &Multiply_LinearColorByFloat
  &ColorFromCurve
  &FNTimeOfDayScalar
  &MakeCustomFloatFromBool
  &SetBoolByFloatComparison
  &FNTimeOfDayColor
 ]
 usedModules [
  &self:SystemSpawnScript.SystemScriptSource.SystemScriptGraph.NiagaraNodeAssignment_0.SetVariables_#1
  &self:SystemSpawnScript.SystemScriptSource.SystemScriptGraph.NiagaraNodeAssignment_1.SetVariables_#2
  &self:SystemSpawnScript.SystemScriptSource.SystemScriptGraph.NiagaraNodeAssignment_2.SetVariables_#3
  &self:SystemSpawnScript.SystemScriptSource.SystemScriptGraph.NiagaraNodeAssignment_3.SetVariables_#4
  &self:SystemSpawnScript.SystemScriptSource.SystemScriptGraph.NiagaraNodeAssignment_4.SetVariables_#5
  &SystemState
 ]
}
emitters {Decal {rendererClasses [], renderers []}, SparklesLWE {rendererClasses [], renderers []}, SparklesLWE001 {rendererClasses [], renderers []}}
summary {
 systemName=NS_Emote_PawJasmine_Sparkles
 emitters [{emitterName=SparklesLWE, rendererClasses []}, {emitterName=SparklesLWE001, rendererClasses []}, {emitterName=Decal, rendererClasses []}]
 userVariables [
  User.Bool {type=NiagaraFloat, defaultValue=1, description=""}
  User.rotation {type=NiagaraFloat, defaultValue=0, description=""}
 ]
}
